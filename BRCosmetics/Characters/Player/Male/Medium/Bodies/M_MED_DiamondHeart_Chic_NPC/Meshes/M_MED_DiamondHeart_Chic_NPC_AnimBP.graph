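# Character-part AnimBlueprint for the DiamondHeart Chic NPC male body, adding secondary motion to cosmetic bones.
# It copies the attached parent's pose, then chains AnimDynamics for necklace, flaps and collars into a RigidBody node using the DiamondHeart Jewel cape physics asset.
# Slide and crouch start/end events reset dynamics.
# Each update sets GravityOverride through a deprecated function, though no anim node visibly reads that variable.
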AnimBlueprint /BRCosmetics/Characters/Player/Male/Medium/Bodies/M_MED_DiamondHeart_Chic_NPC/Meshes/M_MED_DiamondHeart_Chic_NPC_AnimBP
target_skeleton=&/M_MED_DiamondHeart_Chic_NPC_Skeleton
graphs [
 AnimGraph {nodes [
   n1 = Anim Dynamics - Bone: dyn_necklace_01(ComponentPose=Local To Component(LocalPose=Copy Pose From Mesh() {bUseAttachedParent=true}), Alpha=1, ExternalForce="0,0,0") {
    angularDampingOverride=0.9999
    angularSpringConstant=500
    bAngularSpring=true
    bGravityOverrideInSimSpace=true
    bOverrideAngularDamping=true
    simulationSpace=BoneRelative
    alphaBoolBlend {blendOption=Linear}
    boundBone {boneName=dyn_necklace_01}
    componentAppliedLinearAccClamp=<3500 3500 3500>
    componentLinearAccScale=<1.2 1.2 1.2>
    componentLinearVelScale=<1.2 1.2 1.2>
    physicsBodyDefinitions [{
      sphereCollisionRadius=0
      boundBone {boneName=dyn_necklace_01}
      boxExtents=<15 20 20>
      constraintSetup {angularLimitsMax=<1 15 60>, angularLimitsMin=<-1 -15 -5>, angularTarget=<0 0 0>}
      localJointOffset=<4 0 0>
     }]
    relativeSpaceBone {boneName=root}
   }
   n2 = Anim Dynamics - Bone: dyn_necklace_02(ComponentPose=@n1, Alpha=1, ExternalForce="0,0,0") {
    angularDampingOverride=0.9999
    angularSpringConstant=500
    bAngularSpring=true
    bGravityOverrideInSimSpace=true
    bOverrideAngularDamping=true
    simulationSpace=BoneRelative
    alphaBoolBlend {blendOption=Linear}
    boundBone {boneName=dyn_necklace_02}
    componentAppliedLinearAccClamp=<3500 3500 3500>
    componentLinearAccScale=<1.2 1.2 1.2>
    componentLinearVelScale=<1.2 1.2 1.2>
    physicsBodyDefinitions [{
      sphereCollisionRadius=0
      boundBone {boneName=dyn_necklace_02}
      boxExtents=<15 20 20>
      constraintSetup {angularLimitsMax=<1 0 60>, angularLimitsMin=<-1 0 -5>, angularTarget=<0 0 0>}
      localJointOffset=<4 0 0>
     }]
    relativeSpaceBone {boneName=root}
   }
   n3 = Anim Dynamics - Bone: dyn_L_flap_01(ComponentPose=@n2, Alpha=1, ExternalForce="0,0,0") {
    angularDampingOverride=1
    angularSpringConstant=20000
    bAngularSpring=true
    bOverrideAngularDamping=true
    bUseGravityOverride=true
    gravityScale=0.35
    simulationSpace=RootRelative
    alphaBoolBlend {blendOption=Linear}
    boundBone {boneName=dyn_L_flap_01}
    componentLinearAccScale=<1 1 1>
    componentLinearVelScale=<1 1 1>
    gravityOverride {z=-980}
    physicsBodyDefinitions [{
      boundBone {boneName=dyn_L_flap_01}
      boxExtents=<20 30 30>
      constraintSetup {angularLimitsMax=<0 0 20>, angularLimitsMin=<0 0 -5>, angularTarget=<1 0 0>}
      localJointOffset=<4 0 0>
     }]
   }
   n4 = Anim Dynamics - Bone: dyn_R_flap__01(ComponentPose=@n3, Alpha=1, ExternalForce="0,0,0") {
    angularDampingOverride=1
    angularSpringConstant=20000
    bAngularSpring=true
    bOverrideAngularDamping=true
    bUseGravityOverride=true
    gravityScale=0.35
    simulationSpace=RootRelative
    alphaBoolBlend {blendOption=Linear}
    boundBone {boneName=dyn_R_flap__01}
    componentLinearAccScale=<1 1 1>
    componentLinearVelScale=<1 1 1>
    gravityOverride {z=-980}
    physicsBodyDefinitions [{
      boundBone {boneName=dyn_R_flap__01}
      boxExtents=<20 30 30>
      constraintSetup {angularLimitsMax=<0 0 5>, angularLimitsMin=<0 0 -20>, angularTarget=<1 0 0>}
      localJointOffset=<4 0 0>
     }]
   }
   n5 = Anim Dynamics - Bone: dyn_L_innerCollar_A(ComponentPose=@n4, Alpha=1, ExternalForce="0,0,0") {
    angularDampingOverride=1
    bLinearSpring=true
    bOverrideAngularDamping=true
    bOverrideLinearDamping=true
    bUseGravityOverride=true
    gravityScale=0.35
    linearDampingOverride=1
    linearSpringConstant=200
    simulationSpace=BoneRelative
    alphaBoolBlend {blendOption=Linear}
    alphaScaleBiasClamp {scale=0.4}
    boundBone {boneName=dyn_L_innerCollar_A}
    componentLinearAccScale=<0.7 0.7 0.7>
    componentLinearVelScale=<0.7 0.7 0.7>
    gravityOverride {z=-980}
    physicsBodyDefinitions [{
      boundBone {boneName=dyn_L_innerCollar_A}
      boxExtents=<7 7 7>
      constraintSetup {angularLimitsMax=<0 0 0>, angularLimitsMin=<0 0 0>, angularTarget=<1 0 0>, linearAxesMax=<0 0 2>}
      localJointOffset=<0 0 1>
     }]
    relativeSpaceBone {boneName=pelvis}
   }
   n6 = Anim Dynamics - Bone: dyn_L_outerCollar(ComponentPose=@n5, Alpha=1, ExternalForce="0,0,0") {
    angularDampingOverride=1
    angularSpringConstant=20000
    bAngularSpring=true
    bOverrideAngularDamping=true
    bUseGravityOverride=true
    gravityScale=0.35
    simulationSpace=RootRelative
    alphaBoolBlend {blendOption=Linear}
    boundBone {boneName=dyn_L_outerCollar}
    componentLinearAccScale=<1 1 1>
    componentLinearVelScale=<1 1 1>
    gravityOverride {z=-980}
    physicsBodyDefinitions [{
      boundBone {boneName=dyn_L_outerCollar}
      boxExtents=<30 30 30>
      constraintSetup {angularLimitsMax=<5 30 10>, angularLimitsMin=<-5 -30 -10>, angularTarget=<1 0 0>}
      localJointOffset=<3 0 0>
     }]
   }
   n7 = Anim Dynamics - Bone: dyn_R_innerCollar_A(ComponentPose=@n6, Alpha=1, ExternalForce="0,0,0") {
    angularDampingOverride=1
    bLinearSpring=true
    bOverrideAngularDamping=true
    bOverrideLinearDamping=true
    bUseGravityOverride=true
    gravityScale=0.35
    linearDampingOverride=1
    linearSpringConstant=200
    simulationSpace=BoneRelative
    alphaBoolBlend {blendOption=Linear}
    alphaScaleBiasClamp {scale=0.4}
    boundBone {boneName=dyn_R_innerCollar_A}
    componentLinearAccScale=<0.7 0.7 0.7>
    componentLinearVelScale=<0.7 0.7 0.7>
    gravityOverride {z=-980}
    physicsBodyDefinitions [{
      boundBone {boneName=dyn_R_innerCollar_A}
      boxExtents=<7 7 7>
      constraintSetup {angularLimitsMax=<0 0 0>, angularLimitsMin=<0 0 0>, angularTarget=<1 0 0>, linearAxesMin=<0 0 -2>}
      localJointOffset=<0 0 -1>
     }]
    relativeSpaceBone {boneName=pelvis}
   }
   n8 = Anim Dynamics - Bone: dyn_R_outerCollar(ComponentPose=@n7, Alpha=1, ExternalForce="0,0,0") {
    angularDampingOverride=1
    angularSpringConstant=20000
    bAngularSpring=true
    bOverrideAngularDamping=true
    bUseGravityOverride=true
    gravityScale=0.35
    simulationSpace=RootRelative
    alphaBoolBlend {blendOption=Linear}
    boundBone {boneName=dyn_R_outerCollar}
    componentLinearAccScale=<1 1 1>
    componentLinearVelScale=<1 1 1>
    gravityOverride {z=-980}
    physicsBodyDefinitions [{
      boundBone {boneName=dyn_R_outerCollar}
      boxExtents=<30 30 30>
      constraintSetup {angularLimitsMax=<5 30 10>, angularLimitsMin=<-5 -30 -10>, angularTarget=<1 0 0>}
      localJointOffset=<3 0 0>
     }]
   }
   n9 = RigidBody!(ComponentPose=@n8, Alpha, OverrideWorldGravity:Vector="0,0,-980", ExternalForce) -> (Pose) {
    bClampLinearTranslationLimitToRefPose=true
    bOverrideWorldGravity=true
    overridePhysicsAsset=/BRCosmetics/Characters/Player/Male/Medium/Bodies/M_MED_DiamondHeart_Jewel/Meshes/M_MED_DiamondHeart_Jewel_Cape_Physics
    simulationSpace=BaseBoneSpace
    alphaBoolBlend {blendOption=Linear}
    baseBoneRef {boneName=pelvis}
    componentAppliedLinearAccClamp=<1000 1000 1000>
    componentLinearAccScale=<0.5 0.5 0.5>
    componentLinearVelScale=<0.5 0.5 0.5>
    overrideWorldGravity {z=-980}
    simSpaceSettings {
     maxAngularAcceleration=200
     maxAngularVelocity=1
     maxLinearAcceleration=100
     maxLinearVelocity=50
     velocityScaleZ=0.2
     worldAlpha=1
     externalAngularVelocity=<0.2 0.2 0.2>
    }
   }
   Output Pose(Result=Component To Local(ComponentPose=@n9)) {layerGroup=None}
  ]}
 EventGraph {nodes [
   Event BlueprintUpdateAnimation()
   > n1 = Gravity Override State Function DEPRECATED(CharacterPartAnimInstance=Self-Reference(), GameTimeInSeconds=GetGameTimeInSeconds(), GravityOverrideParams=Get Gravity Override Params())
   > Set GravityOverride(GravityOverride=@n1)
   n2 = Event OnEndSliding() {is=title K2Node_Event}
   ResetDynamics(execute=[@n2.then, @n3.then, @n4.then, @n5.then, @n6.then, @n7.then, @n8.then, @n9.then], InTeleportType=ResetPhysics)
   n3 = Event OnBeginKneeSliding() {is=title K2Node_Event}
   n4 = Event OnEndBaseballSliding() {is=title K2Node_Event}
   n5 = Event OnEndCrouching()
   n6 = Event OnBeginBaseballSliding() {is=title K2Node_Event}
   n7 = Event OnBeginCrouching()
   n8 = Event OnEndKneeSliding() {is=title K2Node_Event}
   n9 = Event OnBeginSliding() {is=title K2Node_Event}
   TryGetPawnOwner()
  ]}
]
variables [
 Gravity Override Params {type=GravityOverrideParamsStruct}
 GravityOverride {type=Vector}
]
